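# Unlit translucent star material, used on Niagara mesh particles, for the Pandora gem constellation effect on the gem holder quest prop.
# Two exponential radial gradients form a colored inner star and outer glow, summed, tinted by particle color and star color, then passed through shared emissive adjustments.
# The stronger gradient times particle alpha forms an opacity mask value.
# Rounded UVs become StarSize-scaled corner offsets transformed from view space, a technique for camera-facing quads.
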Material /Game/Environments/Helios/Props/Quest/Quest_GemHolder_A/FX/M_PandoraGem_ConstelationStar
graph {
 properties {
  bUsedWithNiagaraMeshParticles=true
  blendMode=BLEND_Translucent
  pixelDepthOffsetMode=PDOM_Legacy
  shadingModel=MSM_Unlit
  twoSided=true
  usedShadingModels=Unlit
 }
 nodes [
  n1 = VectorParameter() {parameterName=StarColor, defaultValue=<rgba 2.0208 3.8251 4 1>}
  n2 = MaterialFunctionCall(
   Radius=ScalarParameter() {defaultValue=0.2, parameterName=InnerRadius}
   Density=ScalarParameter() {defaultValue=0.5, parameterName=InnerDensity}
  ) {materialFunction=&RadialGradientExponential}
  n3 = Multiply(A=@n1.RGB, B=@n2.RadialGradientExponential)
  n4 = MaterialFunctionCall(Radius=ScalarParameter() {defaultValue=0.6, parameterName=GlowRadius}, Density=ScalarParameter() {defaultValue=1, parameterName=GlowDensity}) {materialFunction=&RadialGradientExponential}
  n5 = Multiply(A=VectorParameter() {parameterName=GlowColor, defaultValue=<rgba 1 0.208 0 1>}.RGB, B=@n4.RadialGradientExponential)
  n6 = ParticleColor()
  MaterialFunctionCall(Emissive=Multiply(A=Add(A=@n3, B=@n5), B=Multiply(A=@n6.RGB, B=@n1.RGB))) {materialFunction=&MF_FX_EmissiveAdjustments}
  Multiply(A=Max(A=@n2.RadialGradientExponential, B=@n4.RadialGradientExponential), B=@n6.A)
  n7 = ScalarParameter() {defaultValue=7, parameterName=StarSize}
  n8 = ConstantBiasScale(None=Round(None=TextureCoordinate())) {bias=-0.5, scale=2}
  Transform(None=AppendVector(A=Multiply(A=@n7, B=@n8), B=Constant())) {transformSourceType=TRANSFORMSOURCE_View}
  VectorParameter() {parameterName=GlowColor, defaultValue=<rgba 1 0.208 0 1>}
 ]
}
