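# Translucent unlit Niagara sprite material that draws a soft, noise-distorted galaxy backdrop for the quest gem holder effect.
# Two panning radial cloud samples distort a capsule mask from Sprite_Capsule UVs, shaping the two-color gradient and opacity.
# Triplanar nebula and star layers are built through MF_PandoraGem_TriplanarBox.
# Shading path switches replace the color overlays and depth fade with constants on forward and mobile.
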Material /Game/Environments/Helios/Props/Quest/Quest_GemHolder_A/FX/M_Quest_GemHolder_GalaxybackGround
graph {
 properties {
  bAutomaticallySetUsageInEditor=false
  bUsedWithNiagaraSprites=true
  blendMode=BLEND_Translucent
  pixelDepthOffsetMode=PDOM_Legacy
  shadingModel=MSM_Unlit
  usedShadingModels=Unlit
 }
 nodes [
  n1 = NamedRerouteUsage()
  n2 = Length(None=Add(A=@n1) {constB=-0.5}) {is=class MaterialExpressionLength}
  n3 = Divide(
   A=Subtract(A=Abs(None=@n2), B=ScalarParameter() {parameterName=InnerColorRadius}) {constB=0.5}
   B=ScalarParameter() {defaultValue=0.2, parameterName=InnerColorSoftness}
  )
  n4 = MaterialFunctionCall(Vector or UVs=Add(A=TextureCoordinate(), B=Constant2Vector() {g=-0.5, r=-0.5})) {materialFunction=&VectorToRadialValue}
  n5 = VectorParameter() {parameterName=DistoNoise_Settings, channelNames {a=PanY, b=PanX, g=ScaleY, r=ScaleX}, defaultValue=<rgba 1 0.5 0 -0.05>}
  n6 = Multiply(
   A=@n4.Radial Coordinates
   B=MaterialFunctionCall(X=@n5.ScaleX, Y=@n5.ScaleX) {materialFunction=&MakeFloat2}.Result
  )
  n7 = Multiply(
   A=MaterialFunctionCall(X=@n5.PanX, Y=@n5.PanX) {materialFunction=&MakeFloat2}.Result
   B=Time()
  )
  n8 = VectorParameter() {parameterName=MicroDistoNoise_Settings, channelNames {a=PanY, b=PanX, g=ScaleY, r=ScaleX}, defaultValue=<rgba 1.5 0.5 0.02 -0.05>}
  n9 = Multiply(
   A=@n4.Radial Coordinates
   B=MaterialFunctionCall(X=@n8.ScaleX, Y=@n8.ScaleX) {materialFunction=&MakeFloat2}.Result
  )
  n10 = Multiply(
   A=MaterialFunctionCall(X=@n8.PanX, Y=@n8.PanX) {materialFunction=&MakeFloat2}.Result
   B=Time()
  )
  n11 = Max(
   A=TextureSample(UVs=Add(A=@n6, B=@n7)) {samplerType=SAMPLERTYPE_LinearGrayscale, texture=/Game/Effects/Fort_Effects/Emotes/UltraLight/Texture/T_Cloud_Greyscale}.RGB
   B=TextureSample(UVs=Add(A=@n9, B=@n10)) {samplerType=SAMPLERTYPE_LinearGrayscale, texture=&T_Cloud_Greyscale}.RGB
  )
  n12 = ConstantBiasScale(None=ComponentMask(None=@n11) {g=true, r=true}) {bias=-0.5, scale=2}
  n13 = ScalarParameter() {defaultValue=0.1, parameterName=DistoStrength}
  n14 = Add(A=Add(A=@n1, B=Multiply(A=@n12, B=@n13) {constB=0})) {constB=-0.5}
  n15 = Subtract(A=Abs(None=Length(None=@n14)), B=ScalarParameter() {defaultValue=0.2, parameterName=CapsuleRadius}) {constB=0.5}
  n16 = ScalarParameter() {defaultValue=0.4, parameterName=CapsuleSoftness}
  n17 = Saturate(None=OneMinus(None=Divide(A=@n15, B=@n16)))
  n18 = ComponentMask(None=@n11) {g=true}
  n19 = Max(A=Saturate(None=OneMinus(None=@n3)), B=Subtract(A=@n17, B=Multiply(A=@n18) {constB=1.5}))
  n20 = NamedRerouteUsage()
  n21 = Constant()
  n22 = Add(
   A=LinearInterpolate(
    A=VectorParameter() {parameterName=Color, defaultValue=<rgba 0.072 0 0.5729 1>}.RGB
    B=VectorParameter() {parameterName=Color_2, defaultValue=<rgba 0.0024 0 0.1615 1>}.RGB
    Alpha=Saturate(None=@n19)
   )
   B=ShadingPathSwitch(Default=@n20, Deferred=@n20, Forward=@n21, Mobile=@n21)
  )
  n23 = VectorParameter() {parameterName=StarColor, defaultValue=<rgba 0.3802 0.8032 1 1>}
  n24 = NamedRerouteUsage()
  n25 = Constant()
  n26 = ShadingPathSwitch(Default=@n24, Deferred=@n24, Forward=@n25, Mobile=@n25)
  n27 = ParticleColor()
  MaterialFunctionCall(Emissive=Multiply(A=LinearInterpolate(A=@n22, B=@n23.RGB, Alpha=@n26), B=@n27.RGB)) {materialFunction=&MF_FX_EmissiveAdjustments}
  n28 = Multiply(A=Saturate(None=@n18), B=ScalarParameter() {defaultValue=0.1, parameterName=NoiseSubtract}) {constB=0.1}
  n29 = Power(Base=SmoothStep(Value=Subtract(A=@n17, B=@n28))) {constExponent=3}
  n30 = ScalarParameter() {defaultValue=0.5, parameterName=GlobalOpacity}
  n31 = DepthFade(FadeDistance=ScalarParameter() {defaultValue=10, parameterName=DepthFadeDist})
  n32 = Constant() {r=1}
  Multiply(A=Multiply(A=@n27.A, B=Multiply(A=@n29, B=@n30)), B=ShadingPathSwitch(Default=@n31, Deferred=@n31, Forward=@n32, Mobile=@n32))
  n33 = DynamicParameter() {paramNames [Particle Direction X, Particle Direction Y, Particle Direction Z, Scale Factor]}
  n34 = MaterialFunctionCall(R=@n33.Particle Direction X, G=@n33.Particle Direction X, B=@n33.Particle Direction X) {materialFunction=&AppendMany}
  NamedRerouteDeclaration(None=MaterialFunctionCall!(
    Particle Position
    Normalized Particle Direction=Transform(None=@n34.RGB) {transformSourceType=TRANSFORMSOURCE_Local}
    Capsule Diameter
    Capsule Pivot Alignment
    Scale Along Velocity Vector=@n33.Scale Factor
    ---------Scale Over Velocity Parameters---------
    Particle Speed
    Minimum Speed
    Max Speed
    Minimum Scale Factor
    Maximum Scale Factor
    ---------- Advanced -----------
    Cap Size Multiplier
    Use Camera Offset
    Use Advanced Features
    World Position to Negate
   ) {materialFunction=/Engine/Functions/Engine_MaterialFunctions02/WorldPositionOffset/Sprite_Capsule}.Texture Coordinates) {name=CapsuleUVs, nodeColor=<rgba 1 0.0356 0 1>}
  n35 = TextureObjectParameter() {
   group=Galaxy
   parameterName=StarsTexture
   samplerType=SAMPLERTYPE_Masks
   sortPriority=80
   texture=/Game/Characters/Player/Male/Medium/Bodies/M_Med_Soldier_04/Skins/Galaxy/Textures/Celestial_T_Stars
  }
  n36 = MaterialFunctionCall!(
   Texture=@n35
   Depth=ScalarParameter() {defaultValue=150, group=Galaxy, parameterName=Stars_Depth, sortPriority=81}
   Tile=ScalarParameter() {defaultValue=1, group=Galaxy, parameterName=Stars_Tile, sortPriority=82}
   BlendContrast=ScalarParameter() {defaultValue=1, group=Galaxy, parameterName=Stars_BlendContrast, sortPriority=83}
   SpeedXY=ComponentMask(None=VectorParameter() {group=Galaxy, parameterName=Stars_Speed, sortPriority=84, defaultValue=<rgba 0 0 0 1>}.RGB) {g=true, r=true}
   RotationSpeed=VectorParameter() {group=Galaxy, parameterName=Stars_RotationSpeed, sortPriority=84, defaultValue=<rgba 0 0 0 1>}.RGB
   ActorOffset=NamedRerouteDeclaration(None=VectorParameter() {group=Galaxy, parameterName=GalaxyOffset, defaultValue=<rgba 0 0 80 1>}.RGB) {name=ActorOffset, nodeColor=<rgba 1 0.0469 0 1>}
  ) {materialFunction=&/MF_PandoraGem_TriplanarBox}
  NamedRerouteDeclaration(None=@n36.UVX) {name=UVX, nodeColor=<rgba 0 0.8828 1 1>}
  NamedRerouteDeclaration(None=@n36.UVY) {name=UVY, nodeColor=<rgba 0 0.8828 1 1>}
  NamedRerouteDeclaration(None=@n36.UVZ) {name=UVZ, nodeColor=<rgba 0 0.8828 1 1>}
  NamedRerouteDeclaration(None=@n36.BlendY) {name=BlendY, nodeColor=<rgba 1 0.6226 0.1707 1>}
  NamedRerouteDeclaration(None=@n36.BlendZ) {name=BlendZ, nodeColor=<rgba 1 0.6226 0.1707 1>}
  n37 = TextureObjectParameter() {
   group=Galaxy
   parameterName=Galaxy_01_Texture
   sortPriority=58
   texture=/Game/Characters/Player/Male/Medium/Bodies/M_Med_Soldier_04/Skins/Galaxy/Textures/T_FN_Nebula_style_blue
  }
  NamedRerouteDeclaration(None=@n37) {name=GalaxyTexture, nodeColor=<rgba 0 1 0.9062 1>}
  NamedRerouteDeclaration(None=Multiply(A=ScalarParameter() {defaultValue=0.6, group=Galaxy, parameterName=Galaxy_01_Tile, sortPriority=62}) {constB=0.01}) {name=GalaxyTile, nodeColor=<rgba 1 0 0.1172 1>}
  n38 = VectorParameter() {group=Galaxy, parameterName=Galaxy_01_Speed, sortPriority=64, defaultValue=<rgba 0.0035 -0.005 0 1>}
  NamedRerouteDeclaration(None=Multiply(A=ComponentMask(None=@n38.RGB) {g=true, r=true}, B=Time())) {name=GalaxySpeed, nodeColor=<rgba 0 0.8125 1 1>}
  n39 = Multiply(A=NamedRerouteUsage(), B=NamedRerouteUsage())
  n40 = NamedRerouteUsage()
  n41 = Multiply(A=NamedRerouteUsage(), B=NamedRerouteUsage())
  n42 = NamedRerouteUsage()
  n43 = LinearInterpolate(
   A=TextureSample(UVs=Add(A=@n39, B=@n40), Tex=NamedRerouteUsage()) {samplerSource=SSM_Wrap_WorldGroupSettings}.RGB
   B=TextureSample(UVs=Add(A=@n41, B=@n42), Tex=NamedRerouteUsage()) {samplerSource=SSM_Wrap_WorldGroupSettings}.RGB
   Alpha=NamedRerouteUsage()
  )
  n44 = NamedRerouteUsage()
  n45 = NamedRerouteUsage()
  n46 = TextureSample(UVs=Add(A=Multiply(A=@n44, B=@n45), B=NamedRerouteUsage()), Tex=NamedRerouteUsage()) {samplerSource=SSM_Wrap_WorldGroupSettings}
  n47 = NamedRerouteUsage()
  NamedRerouteDeclaration(None=Power(
    Base=LinearInterpolate(A=@n43, B=@n46.RGB, Alpha=@n47)
    Exp=ScalarParameter() {defaultValue=5, group=Galaxy, parameterName=Galaxy_01_Exp, sortPriority=66}
   )) {name=Galaxies, nodeColor=<rgba 1 0 0.2578 1>}
  n48 = ChannelMaskParameter(None=NamedRerouteUsage()) {group=Galaxy, maskChannel=Blue, parameterName=StarMaskChannel, sortPriority=84, defaultValue=<rgba 0 0 1 0>}
  NamedRerouteDeclaration(None=Multiply(A=@n48.R, B=ScalarParameter() {defaultValue=3, group=Galaxy, parameterName=GalaxyStarMaskStrength, sortPriority=85}) {constB=2}) {name=StarMask, nodeColor=<rgba 0.9922 1 0 1>}
  n49 = LinearInterpolate(A=NamedRerouteUsage(), Alpha=NamedRerouteUsage()) {constB=0}
  NamedRerouteDeclaration(None=Add(A=Saturate(None=@n49), B=NamedRerouteUsage())) {name=GalaxiesAndStars, nodeColor=<rgba 0 1 0.9062 1>}
  n50 = ComponentMask(None=@n36.Result) {b=true, r=true}
  n51 = Multiply(A=@n50, B=@n50)
  NamedRerouteDeclaration(None=Multiply(
    A=MaterialFunctionCall(f2=@n51) {materialFunction=&AddComponents}.f2
    B=ScalarParameter() {defaultValue=1, group=Galaxy, parameterName=Stars_Brightness, sortPriority=84}
   )) {name=Stars, nodeColor=<rgba 0 1 0.5078 1>}
  NamedRerouteUsage()
 ]
}
