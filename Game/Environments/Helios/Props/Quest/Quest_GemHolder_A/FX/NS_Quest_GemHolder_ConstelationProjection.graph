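# Niagara effect for the quest gem holder that projects a switchable constellation.
# A constellation mesh, a galaxy sprite lerped out to the projection distance, an orange glow with light, and rising soft flares make up the projection.
# System-level CurrentShapeIndex picks mesh, base position and rotation from arrays, while FadeIn and SwitchFade drive scale and alpha.
# User base/next shape indices and switch time are exposed, alongside a custom SwitchCurrentShape module.
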NiagaraSystem /Game/Environments/Helios/Props/Quest/Quest_GemHolder_A/FX/NS_Quest_GemHolder_ConstelationProjection
dependencies {
 usedDataInterfaces [
  &NiagaraDataInterfaceCurve
  &NiagaraDataInterfaceArrayFloat
  &NiagaraDataInterfaceArrayFloat3
  &NiagaraDataInterfaceArrayInt32
 ]
 usedDynamicInputs [
  &SelectIntFromArray
  &MakeIntFromFloat
  &ModuloFloat
  &MakeFloatFromInt
 ]
 usedModules [
  &self:SystemSpawnScript.SystemScriptSource.SystemScriptGraph.NiagaraNodeAssignment_9.SetVariables_#1
  &self:SystemSpawnScript.SystemScriptSource.SystemScriptGraph.NiagaraNodeAssignment_3.SetVariables_#2
  &self:SystemSpawnScript.SystemScriptSource.SystemScriptGraph.NiagaraNodeAssignment_10.SetVariables_#3
  &self:SystemSpawnScript.SystemScriptSource.SystemScriptGraph.NiagaraNodeAssignment_4.SetVariables_#4
  &self:SystemSpawnScript.SystemScriptSource.SystemScriptGraph.NiagaraNodeAssignment_5.SetVariables_#5
  &self:SystemSpawnScript.SystemScriptSource.SystemScriptGraph.NiagaraNodeAssignment_18.SetVariables_#6
  &self:SystemSpawnScript.SystemScriptSource.SystemScriptGraph.NiagaraNodeAssignment_19.SetVariables_#7
  &self:SystemSpawnScript.SystemScriptSource.SystemScriptGraph.NiagaraNodeAssignment_0.SetVariables_#8
  &SystemState
  &self:SystemSpawnScript.SystemScriptSource.SystemScriptGraph.NiagaraNodeAssignment_6.SetVariables_#9
  &self:SystemSpawnScript.SystemScriptSource.SystemScriptGraph.NiagaraNodeAssignment_7.SetVariables_#10
  &self:SystemSpawnScript.SystemScriptSource.SystemScriptGraph.NiagaraNodeAssignment_20.SetVariables_#11
  &self:SystemSpawnScript.SystemScriptSource.SystemScriptGraph.NiagaraNodeAssignment_1.SetVariables_#12
  &Timeline
  &self:SwitchCurrentShape
  &self:SystemSpawnScript.SystemScriptSource.SystemScriptGraph.NiagaraNodeAssignment_17.SetVariables_#13
 ]
}
emitters {
 NE_Constelation {
  emitterUpdateScript [
   &EmitterState(
    Inactive Response=::NewEnumerator1
    Life Cycle Mode=::NewEnumerator1
    Loop Behavior=::NewEnumerator1
    Loop Duration Mode=::NewEnumerator1
    UseLoopDelay=false
   )
   &SpawnBurst_Instantaneous(Use Spawn Probability=false, UseLoopCountLimit=false)
  ]
  particleSpawnScript [
   &PartitionParticles(ID to Partition:NiagaraInt32=&ReturnExecIndex(), Number Of Partitions=$System.StarsNumber)
   &InitializeParticle(
    Lifetime=5
    Mesh Scale Mode=::NewEnumerator3
    Sprite Size Mode=::NewEnumerator3
    Uniform Sprite Size=0
    UsePositionOffset=false
    Write Material Random=true
   )
   &DynamicMaterialParameters(
    Index 0 Param 1=10
    Module.Index0Param0Write=true
    Module.Index0Param1Write=true
    Module.Index0Param2Write=true
    Module.Index0Param3Write=true
   )
   &InitialMeshOrientation(Use Rotation Vector=false)
  ]
  particleUpdateScript [
   &ParticleState(Kill Particles When Lifetime Has Elapsed=false)
   /SetVariables(Particles.Position=&AddVectorToPosition(
     Position=$Engine.Emitter.SimulationPosition
     Vector=&SelectVectorFromArray(Direct Array Index=$System.CurrentShapeIndex, Vector Selection Array=$System.BasePosArray)
    )) {isSetParameters=true}
   &ScaleMeshSize(Scale Factor=&Multiply_VectorByFloat(Vector=<1 1 1>, Float=&Multiply_Float(A=$System.FadeIn, B=$System.SwitchFade)))
   /SetVariables(Particles.MeshOrientation=&MultiplyQuaternion(
     Quaternion A=&MakeQuatFromAxisAngle(
      Axis=&SelectVectorFromArray(Direct Array Index=$System.CurrentShapeIndex, Vector Selection Array=$System.ShapeRotationArray)
      AngleInDegrees=1
     )
     Quaternion B=&MakeQuatFromAxisAngle(
      Axis=<0 0 1>
      AngleInDegrees=&SelectFloatFromArray(Direct Array Index=$System.CurrentShapeIndex, Float Selection Array=DataInterface() {propertyValues {FloatData [180,-90,0,90]}})
     )
    )) {isSetParameters=true}
   /SetVariables(Particles.MeshIndex=$System.CurrentShapeIndex) {isSetParameters=true}
   &ScaleColor(Scale Alpha=$System.SwitchFade, ScaleA=true, ScaleRGB=true)
  ]
  rendererClasses [&NiagaraMeshRendererProperties]
  renderers [{rendererClass=&NiagaraMeshRendererProperties, rendererIndex=0}]
 }
 NE_FakeGalaxy {
  emitterUpdateScript [
   &EmitterState(
    Inactive Response=::NewEnumerator1
    Life Cycle Mode=::NewEnumerator1
    Loop Behavior=::NewEnumerator1
    Loop Duration Mode=::NewEnumerator1
    UseLoopDelay=false
   )
   &SpawnBurst_Instantaneous(Use Spawn Probability=false, UseLoopCountLimit=false)
  ]
  particleSpawnScript [
   &PartitionParticles(ID to Partition:NiagaraInt32=&ReturnExecIndex(), Number Of Partitions=$System.StarsNumber)
   &InitializeParticle(
    Color=<rgba 1 1 1 1.5>
    Lifetime=5
    Sprite Size Mode=::NewEnumerator3
    Uniform Sprite Size=120
    UsePositionOffset=false
    Write Material Random=true
   )
  ]
  particleUpdateScript [
   &ParticleState(Kill Particles When Lifetime Has Elapsed=false)
   /SetVariables(Particles.Position=&AddVectorToPosition(
     Position=&Lerp_Vector(
      A=Linked() {linkedVariable {name=Particles.Initial.Position, type=NiagaraPosition}}
      B=&AddVectorToPosition(
       Position=&RotateVectorByQuaternion(
        Vector To Rotate=Linked() {linkedVariable {name=System.ProjectionDistance, type=NiagaraPosition}}
        Quaternion=&MakeQuatFromAxisAngle(
         Axis=&SelectVectorFromArray(Direct Array Index=$System.CurrentShapeIndex, Vector Selection Array=$System.ShapeRotationArray)
         AngleInDegrees=1
        )
       )
       Vector=<0 0 0>
      )
      Alpha=&Multiply_Float(A=$System.FadeIn, B=$System.SwitchFade)
     )
     Vector=&SelectVectorFromArray(Direct Array Index=$System.CurrentShapeIndex, Vector Selection Array=$System.BasePosArray)
    )) {isSetParameters=true}
   &ScaleSpriteSize(
    Scale Sprite Size Mode=::NewEnumerator0
    Uniform Scale Factor=&Multiply_Float(
     A=&SelectFloatFromArray(Direct Array Index=$System.CurrentShapeIndex, Float Selection Array=DataInterface() {propertyValues {FloatData [1,1.2,1,1.2]}})
     B=&Multiply_Float(A=$System.FadeIn, B=$System.SwitchFade)
    )
   )
   &ScaleColor(Scale Alpha=&Multiply_Float(A=$System.FadeIn, B=$System.SwitchFade), ScaleA=true, ScaleRGB=true)
   &DynamicMaterialParameters(
    Index 0 Param 1=1
    Index 0 Param 4=&Lerp_Float(A=0, B=230, Alpha=&Multiply_Float(A=$System.FadeIn, B=$System.SwitchFade))
    Module.Index0Param0Write=true
    Module.Index0Param1Write=true
    Module.Index0Param2Write=true
    Module.Index0Param3Write=true
   )
  ]
  rendererClasses [&NiagaraSpriteRendererProperties]
  renderers [{rendererClass=&NiagaraSpriteRendererProperties}]
 }
 NE_MainGlow {
  emitterUpdateScript [
   &EmitterState(
    Inactive Response=::NewEnumerator1
    Life Cycle Mode=::NewEnumerator1
    Loop Behavior=::NewEnumerator1
    Loop Duration Mode=::NewEnumerator1
    UseLoopDelay=false
   )
   &SpawnBurst_Instantaneous(Use Spawn Probability=false, UseLoopCountLimit=false)
  ]
  particleSpawnScript [
   &PartitionParticles(Number Of Partitions=$System.StarsNumber)
   &InitializeParticle(
    Color=<rgba 3 0.5706 0 5>
    Lifetime=5
    Sprite Size Mode=::NewEnumerator3
    Uniform Sprite Size=23
    UsePositionOffset=false
    Write Material Random=true
   )
   &DynamicMaterialParameters(Write Parameter Index 0=false)
  ]
  particleUpdateScript [
   &ParticleState(Kill Particles When Lifetime Has Elapsed=false)
   /SetVariables(Particles.Position=&AddVectorToPosition(
     Position=$Engine.Emitter.SimulationPosition
     Vector=&SelectVectorFromArray(Direct Array Index=$System.CurrentShapeIndex, Vector Selection Array=$System.BasePosArray)
    )) {isSetParameters=true}
   &ScaleSpriteSize(
    Scale Sprite Size Mode=::NewEnumerator0
    Uniform Scale Factor=&Multiply_Float(A=$System.FadeIn, B=$System.SwitchFade)
   )
   &ScaleColor(Scale Alpha=&Multiply_Float(A=$System.FadeIn, B=$System.SwitchFade), ScaleA=true, ScaleRGB=true)
   /SetVariables(Particles.GlowSize=&Multiply_Vector2D_ByFloat(Float=2, Vector2D=$Particles.SpriteSize)) {isSetParameters=true}
   &FN_Light_Attributes(
    Color=<rgba 1 0.4322 0 1>
    Create Color Attribute=true
    Create Exponent Attribute=true
    Create Position Attribute=true
    Create Radius Attribute=true
    Exponent=1
    Position=&AddVectorToPosition(Position=$Particles.Position, Vector=<0 0 0>)
    Radius=&Multiply_Float(A=35, B=&Multiply_Float(A=$System.FadeIn, B=$System.SwitchFade))
    Scale Color=50
   )
  ]
  rendererClasses [&NiagaraSpriteRendererProperties, &NiagaraLightRendererProperties]
  renderers [
   {rendererClass=&NiagaraSpriteRendererProperties}
   {rendererClass=&NiagaraSpriteRendererProperties, rendererIndex=1}
   {rendererClass=&NiagaraLightRendererProperties, rendererIndex=2}
  ]
 }
 NE_SoftFlares {
  emitterUpdateScript [
   &EmitterState()
   &SpawnRate(SpawnRate=15, Use Spawn Probability=false)
  ]
  particleSpawnScript [
   &PartitionParticles(
    ID to Partition:NiagaraInt32=&ReturnExecIndex()
    Number Of Partitions=$System.StarsNumber
    Partition Mode:ENiagara_PartitionMode=::NewEnumerator5
   )
   &InitializeParticle(
    Color=<rgba 1 0.2579 0 1>
    Lifetime=2
    Sprite Size=&MakeVector2D(
     X=30
     Y=&Multiply_Float(
      A=&VectorLength(Vector=Linked() {linkedVariable {name=System.ProjectionDistance, type=NiagaraPosition}})
      B=2
     )
    )
    Sprite Size Mode=::NewEnumerator4
    UsePositionOffset=false
    Write Material Random=true
   )
  ]
  particleUpdateScript [
   &ParticleState(Kill Particles When Lifetime Has Elapsed=false)
   /SetVariables(Particles.SpriteAlignment=&RotateVectorByQuaternion(
     Vector To Rotate=&Add_Vector(
      A=<0 0 1.5>
      B=&NormalizeVector(
       Vector To Normalize=&RandomRangeVector(Minimum=<1 0.25 0>, Maximum=<-1 -0.25 1>)
       Safe Normalize=true
       Safe Fallback Threshold=0
       Safe Fallback Vector=<0 0 0.001>
      )
     )
     Quaternion=&MakeQuatFromAxisAngle(
      Axis=&SelectVectorFromArray(Direct Array Index=$System.CurrentShapeIndex, Vector Selection Array=$System.ShapeRotationArray)
      AngleInDegrees=1
     )
    )) {isSetParameters=true}
   /SetVariables(Particles.Position=&AddVectorToPosition(
     Position=$Engine.Emitter.SimulationPosition
     Vector=&SelectVectorFromArray(Direct Array Index=$System.CurrentShapeIndex, Vector Selection Array=$System.BasePosArray)
    )) {isSetParameters=true}
   &ScaleSpriteSize(
    Scale Sprite Size Mode=::NewEnumerator0
    Uniform Scale Factor=&Multiply_Float(
     A=&FloatFromCurve(FloatCurve=DataInterface() {propertyValues {Curve {keys [<at 0 0>,<at 1 1>]}}})
     B=$System.FadeIn
    )
   )
   &ScaleColor(
    Scale Alpha=&Multiply_Float(
     A=&Multiply_Float(
      A=&FloatFromCurve(FloatCurve=DataInterface() {propertyValues {Curve {keys [<at 0 1>,<at 1 0>]}}})
      B=$System.FadeIn
     )
     B=$System.SwitchFade
    )
    ScaleA=true
    ScaleRGB=true
   )
  ]
  rendererClasses [&NiagaraSpriteRendererProperties]
  renderers [{rendererClass=&NiagaraSpriteRendererProperties}]
 }
}
summary {
 systemName=NS_Quest_GemHolder_ConstelationProjection
 emitters [
  {emitterName=NE_MainGlow, rendererClasses [&NiagaraSpriteRendererProperties, &NiagaraLightRendererProperties]}
  {emitterName=NE_FakeGalaxy, rendererClasses [&NiagaraSpriteRendererProperties]}
  {emitterName=NE_SoftFlares, rendererClasses [&NiagaraSpriteRendererProperties]}
  {emitterName=NE_Constelation, rendererClasses [&NiagaraMeshRendererProperties]}
 ]
 userVariables [
  User.BaseShapeIndex {type=NiagaraInt32}
  User.NextShapeIndex {type=NiagaraInt32}
  User.StateSwitchTime {type=NiagaraFloat, defaultValue=0.4, description=""}
 ]
}
